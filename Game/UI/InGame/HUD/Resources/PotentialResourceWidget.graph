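# HUD widget that shows a floating "+N" count for a potential resource item, styled differently for weakpoint hits.
# ShowPotentialResource stores the item and weakpoint flag, updates the text and plays a rise-and-fade animation.
# SetupFont switches font size, material, color and scale animation by that flag, with a ring pulse for weakpoints.
# When the fade ends, or a spectator changes camera or followed player, it hides and broadcasts FinishedShowing.
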WidgetBlueprint /Game/UI/InGame/HUD/Resources/PotentialResourceWidget
animations [
 TranslateAndFadeUp {
  binding_count=1
  display_label=""
  length_seconds=1
  object=&self:TranslateAndFadeUp
  sequence {
   bindings [TextResourceCount {
     id=#1
     possessed_class=TextBlock
     tracks [
      {
       type=MovieScene2DTransformTrack
       sections [{
         type=MovieScene2DTransformSection
         end_frame=60
         start_frame=0
         channels [
          Translation.X {type=MovieSceneScriptingFloatChannel, const [0,1]}
          Translation.Y {type=MovieSceneScriptingFloatChannel, keys [[0,0],[60,-100]]}
          Angle {type=MovieSceneScriptingFloatChannel}
          Scale.X {type=MovieSceneScriptingFloatChannel}
          Scale.Y {type=MovieSceneScriptingFloatChannel}
          Shear.X {type=MovieSceneScriptingFloatChannel}
          Shear.Y {type=MovieSceneScriptingFloatChannel}
         ]
        }]
      }
      {
       type=MovieSceneFloatTrack
       display_name=Render Opacity
       sections [{
         type=MovieSceneFloatSection
         end_frame=60
         start_frame=0
         channels [None {type=MovieSceneScriptingFloatChannel, keys [[0,1],[20.516,1],[60,0]]}]
        }]
      }
     ]
    }]
   display_rate {numerator=60}
  }
 }
 Weakpoint {
  binding_count=1
  display_label=""
  length_seconds=0.2756
  object=&self:Weakpoint
  sequence {
   bindings [TextResourceCount {
     id=#2
     possessed_class=CommonTextBlock
     tracks [{
       type=MovieScene2DTransformTrack
       sections [{
         type=MovieScene2DTransformSection
         end_frame=33
         start_frame=0
         channels [
          Translation.X {type=MovieSceneScriptingFloatChannel}
          Translation.Y {type=MovieSceneScriptingFloatChannel}
          Angle {type=MovieSceneScriptingFloatChannel}
          Scale.X {type=MovieSceneScriptingFloatChannel, keys [[0,1],[16.06,1.4887],[18.86,1.0478],[33.072,1.1162]]}
          Scale.Y {type=MovieSceneScriptingFloatChannel, keys [[0,1],[16.06,1.4887],[18.568,1.0549],[33.072,1.1162]]}
          Shear.X {type=MovieSceneScriptingFloatChannel}
          Shear.Y {type=MovieSceneScriptingFloatChannel}
         ]
        }]
      }]
    }]
   display_rate {numerator=120}
  }
 }
 Nonweakpoint {
  binding_count=1
  display_label=""
  length_seconds=0.05
  object=&self:Nonweakpoint
  sequence {bindings [TextResourceCount {
     id=#3
     possessed_class=CommonTextBlock
     tracks [{
       type=MovieScene2DTransformTrack
       sections [{
         type=MovieScene2DTransformSection
         end_frame=1
         start_frame=0
         channels [
          Translation.X {type=MovieSceneScriptingFloatChannel}
          Translation.Y {type=MovieSceneScriptingFloatChannel}
          Angle {type=MovieSceneScriptingFloatChannel}
          Scale.X {type=MovieSceneScriptingFloatChannel, const [1,1]}
          Scale.Y {type=MovieSceneScriptingFloatChannel, const [1,1]}
          Shear.X {type=MovieSceneScriptingFloatChannel}
          Shear.Y {type=MovieSceneScriptingFloatChannel}
         ]
        }]
      }]
    }]}
 }
 RingPulse {
  binding_count=1
  display_label=""
  length_seconds=0.6162
  object=&self:RingPulse
  sequence {bindings [Image_Ring {
     id=#4
     possessed_class=Image
     tracks [
      {
       type=MovieSceneWidgetMaterialTrack
       sections [{
         type=MovieSceneParameterSection
         end_frame=12
         start_frame=0
         channels [
          RingSize {type=MovieSceneScriptingFloatChannel, keys [[0,0.193],[12.2817,0.03]]}
          EdgeSoftness {type=MovieSceneScriptingFloatChannel, keys [[0,-0.002],[9.6377,-0.062]]}
          ColorR {type=MovieSceneScriptingFloatChannel, const [0,2]}
          ColorG {type=MovieSceneScriptingFloatChannel, keys [[0,0.5984],[6.5373,0]]}
          ColorB {type=MovieSceneScriptingFloatChannel, keys [[0,1.0683],[6.5373,0]]}
          ColorA {type=MovieSceneScriptingFloatChannel, const [1,2]}
         ]
        }]
      }
      {
       type=MovieScene2DTransformTrack
       sections [{
         type=MovieScene2DTransformSection
         end_frame=12
         start_frame=0
         channels [
          Translation.X {type=MovieSceneScriptingFloatChannel}
          Translation.Y {type=MovieSceneScriptingFloatChannel}
          Angle {type=MovieSceneScriptingFloatChannel}
          Scale.X {type=MovieSceneScriptingFloatChannel, keys [[0,0],[12.1843,0.4133]]}
          Scale.Y {type=MovieSceneScriptingFloatChannel, keys [[0,0],[12.1843,0.4133]]}
          Shear.X {type=MovieSceneScriptingFloatChannel}
          Shear.Y {type=MovieSceneScriptingFloatChannel}
         ]
        }]
      }
      {
       type=MovieSceneFloatTrack
       display_name=Render Opacity
       sections [{
         type=MovieSceneFloatSection
         end_frame=9
         start_frame=0
         channels [None {type=MovieSceneScriptingFloatChannel, keys [[0,0],[2.133,0.5749],[9.6377,0]]}]
        }]
      }
     ]
    }]}
 }
]
graphs [
 EventGraph {nodes [
   OnCameraTypeChanged!() -> (OutputDelegate, then, SpectatorPC:Fort Player Controller Spectating Object Reference, NewCameraType:ESpectatorCameraType Enum=ThirdPerson) {is=title K2Node_CustomEvent}
   > HandleTranslateAndFadeUpFinished()
   OnFollowPlayerChanged!() -> (OutputDelegate, then, SpectatorPC:Fort Player Controller Spectating Object Reference, NewFollowedPlayer:Fort Player State Object Reference) {is=title K2Node_CustomEvent}
   > HandleTranslateAndFadeUpFinished()
   HandleTranslateAndFadeUpFinished() {is=title K2Node_CustomEvent}
   > SetVisibility(InVisibility=Hidden)
   > Call Finished Showing!(execute:Exec, self:Potential Resource Widget Object Reference) -> (then:Exec) {is=title K2Node_CallDelegate}
   Event Construct()
   > SetVisibility(InVisibility=Hidden)
   > BindToAnimationFinished(Animation=Get TranslateAndFadeUp() -> (TranslateAndFadeUp:Widget Animation Object Reference), Delegate=Create Event())
   > n1 = Cast To FortPlayerControllerSpectating!(execute, Object=GetOwningPlayer()) -> (then, CastFailed, AsFort Player Controller Spectating:Fort Player Controller Spectating Object Reference) {is=title K2Node_DynamicCast}
   > Bind Event to On Camera Type Changed(self:Fort Player Controller Spectating Object Reference=@n1, Delegate=Create Event()) {is=title K2Node_AddDelegate}
   > Bind Event to On Followed Player Changed(self:Fort Player Controller Spectating Object Reference=@n1, Delegate=Create Event()) {is=title K2Node_AddDelegate}
  ]}
 FinishedShowing {nodes [FinishedShowing() {is=title K2Node_FunctionEntry}]}
 GetResourceCountAsText {nodes [
   n1 = GetResourceCountAsText!() -> (then, Resource:Fort Item Object Reference) {is=title K2Node_FunctionEntry}
   n2 = GetNumInStack!(self:Fort Item Object Reference=@n1) -> (ReturnValue:int=0)
   > Return Node!(execute, Result:Text=Format Text!(Format="LOCTEXT(\"+{ResourceCount}\")", ResourceCount:Text="To Text (Integer)"(Value=@n2)) -> (Result))
  ]}
 SetupFont {nodes [
   n1 = Get Weakpoint() -> (Weakpoint:Widget Animation Object Reference)
   StopAnimation(InAnimation=@n1)
   n2 = Get Nonweakpoint() -> (Nonweakpoint:Widget Animation Object Reference)
   > n3 = StopAnimation(InAnimation=@n2)
   n4 = Get WeakpointDamage() -> (WeakpointDamage:bool)
   PlayAnimation(
    execute=[@n3.then, @n13.then]
    InAnimation=Select(Option 0:Widget Animation Object Reference=@n2, Option 1:Widget Animation Object Reference=@n1, Index=@n4) -> (ReturnValue:Widget Animation Object Reference)
   )
   SetupFont() {is=title K2Node_FunctionEntry}
   > n5 = Set TempFont(TempFont:Slate Font Info Structure=Get Font!(self:Text Object Reference=Get TextResourceCount() -> (TextResourceCount:Common Text Object Reference)) -> (Font:Slate Font Info Structure)) -> (Output_Get:Slate Font Info Structure)
   n6 = Get NormalFontSize() -> (NormalFontSize:int=0)
   n7 = Get WeakpointFontSize() -> (WeakpointFontSize:int=0)
   n8 = Get WeakpointDamage()
   > n9 = Set members in Slate Font Info!(
    execute:Exec
    StructRef:"Slate Font Info Structure (by ref)"=@n5
    Size:float="To Float (Integer)"(InInt=Select(Option 0:int=@n6, Option 1:int=@n7, Index=@n8) -> (ReturnValue:int))
   ) -> (then:Exec, StructOut:"Slate Font Info Structure (by ref)") {is=title K2Node_SetFieldsInStruct}
   > SetFont(self=Get TextResourceCount(), InFontInfo=@n9)
   n10 = Get TextResourceCount()
   n11 = Break Slate Font Info!(SlateFontInfo:"Slate Font Info Structure (by ref)"=Get Font(self=@n10)) -> (
    FontObject:Object Reference
    FontMaterial:Object Reference
    OutlineSettings:Font Outline Settings Structure
    TypefaceFontName:Name
    Size:float=~
    LetterSpacing:int
    SkewAmount:float=~
    bForceMonospaced:bool=~
    bMaterialIsStencil:bool=~
    MonospacedWidth:float=~
   ) {is=title K2Node_BreakStruct}
   n12 = Get WeakpointDamage()
   > n13 = SetFont(
    self=@n10
    InFontInfo=MakeSlateFontInfo!(
     FontObject:Object Reference=@n11
     FontMaterial:Object Reference=Select(
      Option 0:Object Reference=/Game/UI/InGame/HUD/Textures/M_ResourceText
      Option 1:Object Reference=/Game/UI/InGame/HUD/Textures/M_ResourceTextWeakpoint
      Index=@n12
     ) -> (ReturnValue:Object Reference)
     OutlineSettings:Font Outline Settings Structure=@n11.OutlineSettings
     TypefaceFontName:Name=@n11.TypefaceFontName
     Size:float=@n11.Size
     LetterSpacing:int=0
     SkewAmount:float=0
     bForceMonospaced:bool=false
     bMaterialIsStencil:bool=false
     MonospacedWidth:float=1
    ) -> (ReturnValue:Slate Font Info Structure)
   )
   Branch(Condition=@n4)
   > PlayAnimation(InAnimation=Get RingPulse() -> (RingPulse:Widget Animation Object Reference))
   SetColorAndOpacity(InColorAndOpacity=Make Slate Color(SpecifiedColor=Select(Option 0:Linear Color Structure=(R=1,G=1,B=1,A=1), Option 1:Linear Color Structure=(R=0.1534,G=0.5136,B=1,A=1), Index=@n12) -> (ReturnValue:Linear Color Structure)))
  ]}
 ShowPotentialResource {nodes [
   n1 = ShowPotentialResource!() -> (then, PotentialResource:Fort Item Object Reference, bWeakpointDamage:bool=~) {is=title K2Node_FunctionEntry}
   > Set PotentialResource(PotentialResource:Fort Item Object Reference=@n1) -> (Output_Get:Fort Item Object Reference)
   > Set WeakpointDamage(WeakpointDamage:bool=@n1.bWeakpointDamage) -> (Output_Get:bool)
   > n2 = Is Valid(InputObject=Get PotentialResource() -> (PotentialResource:Fort Item Object Reference))
   > n3 = UpdateResourceCountText()
   SetVisibility(execute=[@n3.then, @n2.Is Not Valid], InVisibility=SelfHitTestInvisible)
   > PlayAnimation(InAnimation=Get TranslateAndFadeUp())
  ]}
 UpdateResourceCountText {nodes [
   UpdateResourceCountText() {is=title K2Node_FunctionEntry}
   > SetupFont() {is=title K2Node_CallFunction}
   > "SetText (Text)"(
    self=Get TextResourceCount()
    InText=GetResourceCountAsText!(Resource:Fort Item Object Reference=Get PotentialResource()) -> (Result:Text) {is=title K2Node_CallFunction}
   )
   > Return Node!(execute)
  ]}
]
variables [
 FinishedShowing {}
 NormalFontSize {type=integer, default=16}
 PotentialResource {type=FortItem, default=null}
 WeakpointDamage {type=boolean}
 WeakpointFontSize {type=integer, default=20}
]
widgets {
 info {inheritedWidgetCount=0, parentClass=&CommonUserWidget, rootWidgetClass=&Overlay, widgetCount=3}
 widgets [
  Overlay_0 {widgetClassPath=&Overlay, widget=Overlay_0}
  Image_Ring {
   widgetClassPath=&Image
   parent=Overlay_0
   widget=Image_Ring
   properties {
    renderOpacity=0
    brush {
     bIsDynamicallyLoaded=false
     drawAs=Image
     imageType=NoImage
     mirroring=NoMirror
     resourceName=None
     resourceObject=/Game/UI/InGame/HUD/Textures/M_UI_Ring_Inst
     tiling=NoTile
     imageSize=<256 256>
     margin=<ltrb 0 0 0 0>
     outlineSettings {
      bUseBrushTransparency=false
      roundingType=HalfHeightRadius
      width=0
      color {colorUseRule=UseColor_Specified, specifiedColor=<rgba 0 0 0 0>}
      cornerRadii=<0 0 0 0>
     }
     tintColor {colorUseRule=UseColor_Specified, specifiedColor=<rgba 1 1 1 1>}
     uVRegion {bIsValid=false, max=<0 0>, min=<0 0>}
    }
    renderTransform {angle=0, scale=<0.4133 0.4133>, shear=<0 0>, translation=<3 0>}
   }
   slot {class=OverlaySlot, properties {horizontalAlignment=HAlign_Center, verticalAlignment=VAlign_Center, padding=<ltrb -128 -128 -128 -128>}}
  }
  TextResourceCount {
   widgetClassPath=&CommonTextBlock
   bIsVariable=true
   parent=Overlay_0
   widget=TextResourceCount
   properties {
    bSimpleTextMode=true
    text=+00
    font {
     bForceMonospaced=false
     bMaterialIsStencil=false
     fontMaterial=&M_ResourceText
     fontObject=&NotoSans
     letterSpacing=0
     monospacedWidth=1
     size=18
     skewAmount=0
     typefaceFontName=NotoSans-Bold
     outlineSettings {
      bApplyOutlineToDropShadows=false
      bMiteredCorners=false
      bSeparateFillAlpha=false
      outlineMaterial=null
      outlineSize=0
      outlineColor=<rgba 0 0 0 1>
     }
    }
    renderTransform {angle=0, scale=<1 1>, shear=<-8 0>, translation=<0 0>}
    renderTransformPivot=<0.8 0.8>
    shadowColorAndOpacity=<rgba 0.0052 0.0052 0.0056 0.48>
    shadowOffset=<2.5 2>
   }
   slot {class=OverlaySlot, properties {horizontalAlignment=HAlign_Center, verticalAlignment=VAlign_Center}}
  }
 ]
}
