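# HUD widget that shows a running "+N" total and icon of resources gathered from the building object being damaged.
# Each hit passes the potential resource to a child popup and adds its stack count to the total, which resets when the object changes; weakpoint hits pulse icon and text.
# Destruction arms the fade-out animation, otherwise a restarted timer forces it.
# Spectator camera or followed-player changes call hide, which an ignore flag can skip once.
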WidgetBlueprint /Game/UI/InGame/HUD/Resources/ResourceAggregationWidget
animations [
 AggregateOutroAnim {
  binding_count=2
  display_label=""
  length_seconds=1
  object=&self:AggregateOutroAnim
  sequence {
   bindings [
    TextAggregateResourceCollected {
     id=#1
     possessed_class=CommonTextBlock
     tracks [{
       type=MovieSceneWidgetMaterialTrack
       display_name=Font.FontMaterial
       sections [{
         type=MovieSceneParameterSection
         end_frame=120
         start_frame=0
         channels [ShineOffset {type=MovieSceneScriptingFloatChannel, keys [[0,0.4404],[66.188,-0.9053]]}]
        }]
      }]
    }
    HBoxTotalAggregation {
     id=#2
     possessed_class=HorizontalBox
     tracks [
      {
       type=MovieScene2DTransformTrack
       sections [{
         type=MovieScene2DTransformSection
         end_frame=120
         start_frame=0
         channels [
          Translation.X {type=MovieSceneScriptingFloatChannel}
          Translation.Y {type=MovieSceneScriptingFloatChannel, keys [[0.204,0],[16,8.211],[92.87,-120]]}
          Angle {type=MovieSceneScriptingFloatChannel}
          Scale.X {type=MovieSceneScriptingFloatChannel}
          Scale.Y {type=MovieSceneScriptingFloatChannel, keys [[0,1],[8.41,1.0783],[57.39,1],[120,1]]}
          Shear.X {type=MovieSceneScriptingFloatChannel}
          Shear.Y {type=MovieSceneScriptingFloatChannel}
         ]
        }]
      }
      {
       type=MovieSceneFloatTrack
       display_name=Render Opacity
       sections [{type=MovieSceneFloatSection, start_frame=0, channels [None {type=MovieSceneScriptingFloatChannel, keys [[0,1],[42,1],[95.446,0]]}]}]
      }
     ]
    }
   ]
   display_rate {numerator=120}
  }
 }
 ResourceIconPulseAnim {
  binding_count=4
  display_label=""
  length_seconds=0.35
  object=&self:ResourceIconPulseAnim
  sequence {
   bindings [
    ImageResourceIcon {
     id=#3
     possessed_class=Image
     tracks [{
       type=MovieScene2DTransformTrack
       sections [{
         type=MovieScene2DTransformSection
         end_frame=35
         start_frame=0
         channels [
          Translation.X {type=MovieSceneScriptingFloatChannel, default=0}
          Translation.Y {type=MovieSceneScriptingFloatChannel, default=0}
          Angle {type=MovieSceneScriptingFloatChannel, default=0}
          Scale.X {type=MovieSceneScriptingFloatChannel, keys [[0,1],[7.0483,1.5],[30.0933,0.9574],[34.9983,1]]}
          Scale.Y {type=MovieSceneScriptingFloatChannel, keys [[0,1],[7.0483,1.5],[30.0933,0.9574],[34.9983,1]]}
          Shear.X {type=MovieSceneScriptingFloatChannel, default=0}
          Shear.Y {type=MovieSceneScriptingFloatChannel, default=0}
         ]
        }]
      }]
    }
    Overlay_Text {
     id=#4
     possessed_class=Overlay
     tracks [{
       type=MovieScene2DTransformTrack
       sections [{
         type=MovieScene2DTransformSection
         end_frame=35
         start_frame=0
         channels [
          Translation.X {type=MovieSceneScriptingFloatChannel}
          Translation.Y {type=MovieSceneScriptingFloatChannel}
          Angle {type=MovieSceneScriptingFloatChannel}
          Scale.X {type=MovieSceneScriptingFloatChannel, keys [[0,1],[7.0483,1.3],[30.0933,0.9456],[34.9983,1]]}
          Scale.Y {type=MovieSceneScriptingFloatChannel, keys [[0,1],[7.0483,1.3],[30.0933,0.9456],[34.9983,1]]}
          Shear.X {type=MovieSceneScriptingFloatChannel}
          Shear.Y {type=MovieSceneScriptingFloatChannel}
         ]
        }]
      }]
    }
    Image_Glow {
     id=#5
     possessed_class=Image
     tracks [{
       type=MovieSceneFloatTrack
       display_name=Render Opacity
       sections [{
         type=MovieSceneFloatSection
         end_frame=35
         start_frame=-1
         channels [None {type=MovieSceneScriptingFloatChannel, keys [[0,0],[9.4433,1],[24.3883,0]]}]
        }]
      }]
    }
    TextAggregateResourceCollected {
     id=#6
     possessed_class=CommonTextBlock
     tracks [{
       type=MovieSceneWidgetMaterialTrack
       display_name=Font.FontMaterial
       sections [{
         type=MovieSceneParameterSection
         end_frame=35
         start_frame=-1
         channels [ShineOffset {type=MovieSceneScriptingFloatChannel, keys [[0,0.3786],[35,-0.5889]]}]
        }]
      }]
    }
   ]
   display_rate {numerator=100}
  }
 }
]
graphs [
 EventGraph {nodes [
   OnCameraTypeChanged!() -> (OutputDelegate, then, SpectatorPC:Fort Player Controller Spectating Object Reference, NewCameraType:ESpectatorCameraType Enum=ThirdPerson) {is=title K2Node_CustomEvent}
   > HideWidget()
   OnFollowPlayerChanged!() -> (OutputDelegate, then, SpectatorPC:Fort Player Controller Spectating Object Reference, NewFollowedPlayer:Fort Player State Object Reference) {is=title K2Node_CustomEvent}
   > HideWidget()
   Event Construct()
   > HideWidgetElementsViaAnimation()
   > HideWidget()
   > Bind Event to On Damaged Resource Building(
    self:Fort HUDContext Object Reference=Get FortHUDContext!() -> (ReturnValue:Fort HUDContext Object Reference) {is=title K2Node_GetBlueprintContextEx}
    Delegate=Create Event()
   ) {is=title K2Node_AddDelegate}
   > BindToAnimationFinished(Animation=Get AggregateOutroAnim() -> (AggregateOutroAnim:Widget Animation Object Reference), Delegate=Create Event())
   > Bind Event to Finished Showing(
    self:Potential Resource Widget Object Reference=Get PotentialResourceWidget() -> (PotentialResourceWidget:Potential Resource Widget Object Reference)
    Delegate=Create Event()
   ) {is=title K2Node_AddDelegate}
   > n1 = Cast To FortPlayerControllerSpectating!(execute, Object=GetOwningPlayer()) -> (then, CastFailed, AsFort Player Controller Spectating:Fort Player Controller Spectating Object Reference) {is=title K2Node_DynamicCast}
   > Bind Event to On Camera Type Changed(self:Fort Player Controller Spectating Object Reference=@n1, Delegate=Create Event()) {is=title K2Node_AddDelegate}
   > Bind Event to On Followed Player Changed(self:Fort Player Controller Spectating Object Reference=@n1, Delegate=Create Event()) {is=title K2Node_AddDelegate}
  ]}
 ForceHideWidget {nodes [
   ForceHideWidget() {is=title K2Node_FunctionEntry}
   > Set bActivateAggregateAnim(bActivateAggregateAnim:bool=true) -> (Output_Get:bool)
   > TriggerAggregateAnimation()
  ]}
 HandleDamagedResourceObject {nodes [
   n1 = HandleDamagedResourceObject!() -> (
    then
    DamagedResourceObject:Building Static Mesh Object Reference
    PotentialResource:Fort Item Object Reference
    Destroyed:bool=~
    WeakpointDamage:bool=~
   ) {is=title K2Node_FunctionEntry}
   > Set LocalDamagedResourceObject(LocalDamagedResourceObject:Building Static Mesh Object Reference=@n1) -> (Output_Get:Building Static Mesh Object Reference)
   > Set LocalPotentialResource(LocalPotentialResource:Fort Item Object Reference=@n1.PotentialResource) -> (Output_Get:Fort Item Object Reference)
   > Set LocalDestroyed(LocalDestroyed:bool=@n1.Destroyed) -> (Output_Get:bool)
   > Set LocalWeakpointDamage(LocalWeakpointDamage:bool=@n1.WeakpointDamage) -> (Output_Get:bool)
   > n2 = Branch(Condition="Equal (Object)"(
     A=Get ResourceHoldingObject() -> (ResourceHoldingObject:Building Static Mesh Object Reference)
     B=Get LocalDamagedResourceObject() -> (LocalDamagedResourceObject:Building Static Mesh Object Reference)
    ) {is=title K2Node_CallFunction})
   > n3 = UpdateAggregateResourceCollected!(
    execute
    self
    InPotentialResource:Fort Item Object Reference=Get LocalPotentialResource() -> (LocalPotentialResource:Fort Item Object Reference)
   ) -> (then)
   n4 = Branch(execute=[@n3.then, @n9.then], Condition=Get LocalWeakpointDamage() -> (LocalWeakpointDamage:bool))
   > n5 = PlayAnimation(InAnimation=Get ResourceIconPulseAnim() -> (ResourceIconPulseAnim:Widget Animation Object Reference))
   ShowPotentialResource!(
    execute=[@n5.then, @n4.else]
    self:Potential Resource Widget Object Reference=Get PotentialResourceWidget()
    PotentialResource:Fort Item Object Reference=Get LocalPotentialResource()
    bWeakpointDamage:bool=Get LocalWeakpointDamage()
   ) -> (then)
   n6 = Get AggregateResourceCollected() -> (AggregateResourceCollected:int=0)
   > "SetText (Text)"(
    self=Get TextAggregateResourceCollected() -> (TextAggregateResourceCollected:Common Text Object Reference)
    InText=Format Text!(Format="LOCTEXT(\"+{TotalResourcesCollected}\")", TotalResourcesCollected:Text="To Text (Integer)"(Value=@n6)) -> (Result)
   )
   > SetVisibility(InVisibility=SelfHitTestInvisible)
   > SetVisibility(self=Get HBoxTotalAggregation() -> (HBoxTotalAggregation:Horizontal Box Object Reference), InVisibility=SelfHitTestInvisible)
   > Clear and Invalidate Timer by Handle(Handle=Get HideWidgetTimer() -> (HideWidgetTimer:Timer Handle Structure))
   > n7 = Branch(Condition=Get LocalDestroyed() -> (LocalDestroyed:bool))
   > Set bActivateAggregateAnim(bActivateAggregateAnim=true)
   n8 = Set Timer by Function Name!(
    execute=@n7.else
    Object:Object Reference
    FunctionName:String=ForceHideWidget
    Time:float=3
    bLooping:bool=false
    bMaxOncePerFrame:bool=false
    InitialStartDelay:float=0
    InitialStartDelayVariance:float=0
   ) -> (then, ReturnValue:Timer Handle Structure) {is=title K2Node_CallFunction}
   > Set HideWidgetTimer(HideWidgetTimer:Timer Handle Structure=@n8) -> (Output_Get:Timer Handle Structure)
   n9 = ResetSourceOfResource!(
    execute=@n2.else
    self
    InResourceObject:Building Static Mesh Object Reference=Get LocalDamagedResourceObject()
    InPotentialResource:Fort Item Object Reference=Get LocalPotentialResource()
   ) -> (then)
  ]}
 HideWidget {nodes [
   HideWidget() {is=title K2Node_FunctionEntry}
   > n1 = Branch(Condition=Get bIgnoreHide() -> (bIgnoreHide:bool))
   > Set bIgnoreHide(bIgnoreHide:bool=false) -> (Output_Get:bool)
   SetVisibility(execute=@n1.else, InVisibility=Collapsed)
   > Set ResourceHoldingObject(ResourceHoldingObject:Building Static Mesh Object Reference) -> (Output_Get:Building Static Mesh Object Reference)
  ]}
 HideWidgetElementsViaAnimation {nodes [
   HideWidgetElementsViaAnimation() {is=title K2Node_FunctionEntry}
   > PlayAnimation(InAnimation=Get AggregateOutroAnim(), StartAtTime=1)
  ]}
 ResetSourceOfResource {nodes [
   n1 = ResetSourceOfResource!() -> (then, InResourceObject:Building Static Mesh Object Reference, InPotentialResource:Fort Item Object Reference) {is=title K2Node_FunctionEntry}
   > Set ResourceHoldingObject(ResourceHoldingObject=@n1)
   > SetBrush(
    self=Get ImageResourceIcon() -> (ImageResourceIcon:Image Object Reference)
    InBrush=GetItemSmallPreviewImageBrush!(Item:Fort Item Object Reference=@n1.InPotentialResource, BrushSize:EFortBrushSize Enum=Medium) -> (ReturnValue:Slate Brush Structure)
   )
   > Set AggregateResourceCollected(AggregateResourceCollected:int=GetNumInStack!(self:Fort Item Object Reference=@n1.InPotentialResource) -> (ReturnValue:int=0)) -> (Output_Get:int=0)
   > Set bActivateAggregateAnim(bActivateAggregateAnim=false)
   > Set bIgnoreHide(bIgnoreHide=true)
   > PlayAnimationTimeRange!(
    execute
    self:User Widget Object Reference
    InAnimation:Widget Animation Object Reference=Get AggregateOutroAnim()
    StartAtTime:float=0
    EndAtTime:float=.000001
    NumLoopsToPlay:int=1
    PlayMode:EUMGSequencePlayMode Enum
    PlaybackSpeed:float=1
    bRestoreState:bool=false
   ) -> (then, ReturnValue:Widget Animation Handle Structure)
  ]}
 TriggerAggregateAnimation {nodes [
   TriggerAggregateAnimation() {is=title K2Node_FunctionEntry}
   > Branch(Condition=Get bActivateAggregateAnim() -> (bActivateAggregateAnim:bool))
   > Set bActivateAggregateAnim(bActivateAggregateAnim=false)
   > Set bIgnoreHide(bIgnoreHide=false)
   > PlayAnimation(InAnimation=Get AggregateOutroAnim())
  ]}
 UpdateAggregateResourceCollected {nodes [
   n1 = UpdateAggregateResourceCollected!() -> (then, InPotentialResource:Fort Item Object Reference) {is=title K2Node_FunctionEntry}
   > Set AggregateResourceCollected(AggregateResourceCollected=int + int(A=Get AggregateResourceCollected() -> (AggregateResourceCollected=0), B=GetNumInStack(self=@n1) -> (ReturnValue=0)) -> (ReturnValue=0) {is=title K2Node_CommutativeAssociativeBinaryOperator}) -> (Output_Get=0)
  ]}
]
variables [
 AggregateResourceCollected {type=integer}
 bActivateAggregateAnim {type=boolean}
 bIgnoreHide {type=boolean}
 HideWidgetTimer {type=TimerHandle}
 ResourceHoldingObject {type=BuildingSMActor, default=null}
]
widgets {
 info {inheritedWidgetCount=1, parentClass=&FortHUDElementWidget, rootWidgetClass=&SizeBox, widgetCount=9}
 widgets [
  SizeBox_0 {
   widgetClassPath=&SizeBox
   widget=SizeBox_0
   properties {bOverride_MaxDesiredHeight=true, bOverride_MaxDesiredWidth=true, maxDesiredHeight=300, maxDesiredWidth=200, minDesiredHeight=300}
  }
  VerticalBox_0 {widgetClassPath=&VerticalBox, parent=SizeBox_0, widget=VerticalBox_0, slot {class=SizeBoxSlot}}
  PotentialResourceWidget {
   widgetClassPath=&/PotentialResourceWidget
   bIsVariable=true
   parent=VerticalBox_0
   widget=PotentialResourceWidget
   properties {renderTransform {angle=0, scale=<1.2 1.2>, shear=<0 0>, translation=<0 0>}}
   slot {class=VerticalBoxSlot, properties {horizontalAlignment=HAlign_Right}}
  }
  HBoxTotalAggregation {
   widgetClassPath=&HorizontalBox
   bIsVariable=true
   parent=VerticalBox_0
   widget=HBoxTotalAggregation
   properties {renderTransformPivot=<0.5 0.1>}
   slot {class=VerticalBoxSlot, properties {horizontalAlignment=HAlign_Right}}
  }
  ImageResourceIcon {
   widgetClassPath=&Image
   bIsVariable=true
   parent=HBoxTotalAggregation
   widget=ImageResourceIcon
   properties {
    brush {
     bIsDynamicallyLoaded=false
     drawAs=Image
     imageType=NoImage
     mirroring=NoMirror
     resourceName=None
     resourceObject=/Game/UI/Icons/Icon-Resource/Icon-Resource-Wood-M
     tiling=NoTile
     imageSize=<40 40>
     margin=<ltrb 0 0 0 0>
     outlineSettings {
      bUseBrushTransparency=false
      roundingType=HalfHeightRadius
      width=0
      color {colorUseRule=UseColor_Specified, specifiedColor=<rgba 0 0 0 0>}
      cornerRadii=<0 0 0 0>
     }
     tintColor {colorUseRule=UseColor_Specified, specifiedColor=<rgba 1 1 1 1>}
     uVRegion {bIsValid=false, max=<0 0>, min=<0 0>}
    }
    renderTransformPivot=<0.8 0.7>
   }
   slot {class=HorizontalBoxSlot, properties {horizontalAlignment=HAlign_Center, verticalAlignment=VAlign_Center}}
  }
  Overlay_Text {
   widgetClassPath=&Overlay
   parent=HBoxTotalAggregation
   widget=Overlay_Text
   properties {renderTransform {angle=0, scale=<1 1>, shear=<-10 0>, translation=<0 0>}, renderTransformPivot=<0.2 0.7>}
   slot {class=HorizontalBoxSlot, properties {verticalAlignment=VAlign_Center}}
  }
  TextAggregateResourceCollected {
   widgetClassPath=&CommonTextBlock
   bIsVariable=true
   parent=Overlay_Text
   widget=TextAggregateResourceCollected
   properties {
    bSimpleTextMode=true
    text=100
    font {
     bForceMonospaced=false
     bMaterialIsStencil=false
     fontMaterial=/Game/UI/InGame/HUD/Textures/M_ResourceText
     fontObject=&NotoSans
     letterSpacing=0
     monospacedWidth=1
     size=32
     skewAmount=0
     typefaceFontName=NotoSans-Bold
     outlineSettings {
      bApplyOutlineToDropShadows=false
      bMiteredCorners=false
      bSeparateFillAlpha=false
      outlineMaterial=null
      outlineSize=0
      outlineColor=<rgba 0 0 0 1>
     }
    }
    shadowColorAndOpacity=<rgba 0 0 0 1>
    shadowOffset=<3 5>
   }
   slot {class=OverlaySlot}
  }
  Image_Glow {
   widgetClassPath=&Image
   parent=Overlay_Text
   widget=Image_Glow
   properties {
    renderOpacity=0
    brush {
     bIsDynamicallyLoaded=false
     drawAs=Image
     imageType=NoImage
     mirroring=NoMirror
     resourceName=None
     resourceObject=/Game/Athena/HUD/Materials/M_CountdownBG_SolidGlow
     tiling=NoTile
     imageSize=<170.1995 93.5933>
     margin=<ltrb 0 0 0 0>
     outlineSettings {
      bUseBrushTransparency=false
      roundingType=HalfHeightRadius
      width=0
      color {colorUseRule=UseColor_Specified, specifiedColor=<rgba 0 0 0 0>}
      cornerRadii=<0 0 0 0>
     }
     tintColor {colorUseRule=UseColor_Specified, specifiedColor=<rgba 0 0.2692 1 1>}
     uVRegion {bIsValid=false, max=<0 0>, min=<0 0>}
    }
    colorAndOpacity=<rgba 1 1 1 0.1429>
    renderTransform {angle=0, scale=<1 1>, shear=<-10 0>, translation=<2 0>}
   }
   slot {class=OverlaySlot, properties {horizontalAlignment=HAlign_Center, verticalAlignment=VAlign_Center, padding=<ltrb -250 -250 -250 -250>}}
  }
  TutorialHighlight {widgetClassPath=&TutorialHighlight}
 ]
}
