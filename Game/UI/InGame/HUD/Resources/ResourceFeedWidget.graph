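# HUD widget derived from FortResourceFeed that shows resource gains as an icon, running count, gained amount and weakpoint glow.
# Parent-class events play named ResourceGained, WeakpointHit and FadeOut animations on a curve-driven FortSimpleWidgetAnimationsPanel, whose finish callback switches on name to call the parent's matching handlers.
# For spectator controllers, Construct binds camera-type and followed-player delegates; those custom events collapse the widget.
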WidgetBlueprint /Game/UI/InGame/HUD/Resources/ResourceFeedWidget
graphs [EventGraph {nodes [
   n1 = OnCameraTypeChanged!() -> (OutputDelegate, then, SpectatorPC:Fort Player Controller Spectating Object Reference, NewCameraType:ESpectatorCameraType Enum=ThirdPerson) {is=title K2Node_CustomEvent}
   SetVisibility(execute=[@n1.then, @n2.then], InVisibility=Collapsed)
   n2 = OnFollowPlayerChanged!() -> (OutputDelegate, then, SpectatorPC:Fort Player Controller Spectating Object Reference, NewFollowedPlayer:Fort Player State Object Reference) {is=title K2Node_CustomEvent}
   n3 = OnAnimationsFinished!() -> (OutputDelegate, then, Name:Name) {is=title K2Node_CustomEvent}
   > n4 = Switch on Name!(execute:Exec, Selection:Name=@n3.Name) -> (ResourceGained:Exec, WeakpointHit:Exec, FadeOut:Exec) {is=title K2Node_SwitchName, bHasDefaultPin=false, pinNames [ResourceGained, WeakpointHit, FadeOut]}
   HandleFadeOutFinished(execute=@n4.FadeOut, self:Fort Resource Feed Object Reference)
   HandleResourceGainedFinished(execute=@n4.ResourceGained, self:Fort Resource Feed Object Reference)
   HandleWeakpointHitFinished(execute=@n4.WeakpointHit, self:Fort Resource Feed Object Reference)
   Event Construct()
   > n5 = Sequence()
   > n6 = Cast To FortPlayerControllerSpectating!(execute, Object=GetOwningPlayer()) -> (then, CastFailed, AsFort Player Controller Spectating:Fort Player Controller Spectating Object Reference) {is=title K2Node_DynamicCast}
   > Bind Event to On Camera Type Changed(self:Fort Player Controller Spectating Object Reference=@n6, Delegate=Create Event()) {is=title K2Node_AddDelegate}
   > Bind Event to On Followed Player Changed(self:Fort Player Controller Spectating Object Reference=@n6, Delegate=Create Event()) {is=title K2Node_AddDelegate}
   Bind Event to On Animations Finished(
    execute=@n5.then_1
    self:Fort Simple Widget Animations Panel Object Reference=Get Panel_Animations() -> (Panel_Animations:Fort Simple Widget Animations Panel Object Reference)
    Delegate=Create Event()
   ) {is=title K2Node_AddDelegate}
   Event OnFadeOut() {is=title K2Node_Event}
   > PlayAnimationByName!(
    execute
    self:Fort Simple Widget Animations Panel Object Reference=Get Panel_Animations()
    Name:"Name (by ref)"=Get FadeOutAnimation() -> (FadeOutAnimation:Name=None)
    Speed:float=1
    StartAtTime:float=0
   ) -> (then, ReturnValue:bool)
   Event OnResourceGained() {is=title K2Node_Event}
   > PlayAnimationByName(self=Get Panel_Animations(), Name=Get ResourceGainedAnimation() -> (ResourceGainedAnimation:Name=None), Speed=1, StartAtTime=0)
   Event OnWeakpointHit() {is=title K2Node_Event}
   > PlayAnimationByName(self=Get Panel_Animations(), Name=Get WeakpointHitAnimation() -> (WeakpointHitAnimation:Name=None), Speed=1, StartAtTime=0)
  ]}]
variables [
 FadeOutAnimation {type=string, default=FadeOut}
 ResourceGainedAnimation {type=string, default=ResourceGained}
 WeakpointHitAnimation {type=string, default=WeakpointHit}
]
widgets {
 info {
  inheritedWidgetCount=8
  parentClass=/Script/FortniteUI.FortResourceFeed
  rootWidgetClass=/Script/FortniteUI.FortSimpleWidgetAnimationsPanel
  widgetCount=16
 }
 widgets [
  Panel_Animations {
   widgetClassPath=&FortSimpleWidgetAnimationsPanel
   bInherited=true
   bIsVariable=true
   widget=Panel_Animations
   properties {animations [
     ResourceGained {
      bIsEnabled=true
      onAnimationsFinished=()
      targets [{
        alpha=/Game/Athena/HUD/Art/ResourceFeed/ResourceGained_Alpha
        bResetOnFinish=true
        bShouldAnimateAlpha=true
        bShouldAnimateScale=true
        bShouldAnimateTranslation=true
        scale=null
        targetName=Widget_ResourceGained
        translation=/Game/Athena/HUD/Art/ResourceFeed/ResourceGained_Translation
        materialScalarAnimations []
       }]
     }
     WeakpointHit {
      bIsEnabled=true
      onAnimationsFinished=()
      targets [
       {
        alpha=/Game/Athena/HUD/Art/ResourceFeed/WeakpointHit_Glow_Alpha
        bResetOnFinish=true
        bShouldAnimateAlpha=true
        bShouldAnimateScale=false
        bShouldAnimateTranslation=false
        scale=null
        targetName=Widget_Glow
        translation=null
        materialScalarAnimations []
       }
       {
        alpha=null
        bResetOnFinish=true
        bShouldAnimateAlpha=false
        bShouldAnimateScale=true
        bShouldAnimateTranslation=false
        scale=/Game/Athena/HUD/Art/ResourceFeed/WeakpointHit_Count_Scale
        targetName=Widget_ResourceCount
        translation=null
        materialScalarAnimations []
       }
      ]
     }
     FadeOut {
      bIsEnabled=true
      onAnimationsFinished=()
      targets [{
        alpha=/Game/Athena/HUD/Art/ResourceFeed/FadeOut_Alpha
        bResetOnFinish=true
        bShouldAnimateAlpha=true
        bShouldAnimateScale=true
        bShouldAnimateTranslation=true
        scale=/Game/Athena/HUD/Art/ResourceFeed/FadeOut_Scale
        targetName=Widget_ResourceCount
        translation=/Game/Athena/HUD/Art/ResourceFeed/FadeOut_Translation
        materialScalarAnimations []
       }]
     }
    ]}
  }
  SBoxRoot {
   widgetClassPath=&SizeBox
   parent=Panel_Animations
   widget=SBoxRoot
   properties {bOverride_MaxDesiredHeight=true, bOverride_MaxDesiredWidth=true, maxDesiredHeight=300, maxDesiredWidth=200, minDesiredHeight=300}
   slot {class=FortSimpleWidgetAnimationsSlot}
  }
  VerticalBox_0 {widgetClassPath=&VerticalBox, parent=SBoxRoot, widget=VerticalBox_0, slot {class=SizeBoxSlot}}
  Widget_ResourceGained {
   widgetClassPath=&Overlay
   bInherited=true
   bIsVariable=true
   parent=VerticalBox_0
   widget=Widget_ResourceGained
   properties {renderTransform {angle=0, scale=<1.2 1.2>, shear=<0 0>, translation=<0 0>}}
   slot {class=VerticalBoxSlot, properties {horizontalAlignment=HAlign_Right}}
  }
  HorizontalBox_2 {
   widgetClassPath=&HorizontalBox
   parent=Widget_ResourceGained
   widget=HorizontalBox_2
   slot {class=OverlaySlot, properties {horizontalAlignment=HAlign_Right, verticalAlignment=VAlign_Center}}
  }
  Text_GainedSign {
   widgetClassPath=&CommonTextBlock
   parent=HorizontalBox_2
   widget=Text_GainedSign
   properties {
    bSimpleTextMode=true
    text=+
    font {
     bForceMonospaced=false
     bMaterialIsStencil=false
     fontMaterial=null
     fontObject=&Burbank
     letterSpacing=0
     monospacedWidth=1
     size=20
     skewAmount=0
     typefaceFontName=Burbank-PlayerPower
     outlineSettings {
      bApplyOutlineToDropShadows=false
      bMiteredCorners=false
      bSeparateFillAlpha=false
      outlineMaterial=null
      outlineSize=0
      outlineColor=<rgba 0 0 0 1>
     }
    }
    renderTransform {angle=0, scale=<1 1>, shear=<-8 0>, translation=<0 0>}
    renderTransformPivot=<0.8 0.8>
    shadowColorAndOpacity=<rgba 0.0052 0.0052 0.0056 0.48>
    shadowOffset=<2.5 2>
   }
   slot {class=HorizontalBoxSlot, properties {horizontalAlignment=HAlign_Left, verticalAlignment=VAlign_Center}}
  }
  Text_ResourceGained {
   widgetClassPath=&CommonTextBlock
   bInherited=true
   parent=HorizontalBox_2
   widget=Text_ResourceGained
   properties {
    bSimpleTextMode=true
    text=00
    font {
     bForceMonospaced=false
     bMaterialIsStencil=false
     fontMaterial=/Game/UI/InGame/HUD/Textures/M_ResourceText
     fontObject=&Burbank
     letterSpacing=0
     monospacedWidth=1
     size=20
     skewAmount=0
     typefaceFontName=Burbank-Black
     outlineSettings {
      bApplyOutlineToDropShadows=false
      bMiteredCorners=false
      bSeparateFillAlpha=false
      outlineMaterial=null
      outlineSize=0
      outlineColor=<rgba 0 0 0 1>
     }
    }
    renderTransform {angle=0, scale=<1 1>, shear=<-8 0>, translation=<0 0>}
    renderTransformPivot=<0.8 0.8>
    shadowColorAndOpacity=<rgba 0.0052 0.0052 0.0056 0.48>
    shadowOffset=<2.5 2>
   }
   slot {class=HorizontalBoxSlot, properties {horizontalAlignment=HAlign_Right, verticalAlignment=VAlign_Center}}
  }
  Widget_ResourceCount {
   widgetClassPath=&HorizontalBox
   bInherited=true
   bIsVariable=true
   parent=VerticalBox_0
   widget=Widget_ResourceCount
   properties {bIsVolatile=true, renderTransformPivot=<0.5 0.1>}
   slot {class=VerticalBoxSlot, properties {horizontalAlignment=HAlign_Right}}
  }
  Image_ResourceIcon {
   widgetClassPath=&Image
   bInherited=true
   parent=Widget_ResourceCount
   widget=Image_ResourceIcon
   properties {
    visibility=SelfHitTestInvisible
    brush {
     bIsDynamicallyLoaded=false
     drawAs=Image
     imageType=NoImage
     mirroring=NoMirror
     resourceName=None
     resourceObject=/Game/UI/Icons/Icon-Resource/Icon-Resource-Wood-M
     tiling=NoTile
     imageSize=<40 40>
     margin=<ltrb 0 0 0 0>
     outlineSettings {
      bUseBrushTransparency=false
      roundingType=HalfHeightRadius
      width=0
      color {colorUseRule=UseColor_Specified, specifiedColor=<rgba 0 0 0 0>}
      cornerRadii=<0 0 0 0>
     }
     tintColor {colorUseRule=UseColor_Specified, specifiedColor=<rgba 1 1 1 1>}
     uVRegion {bIsValid=false, max=<0 0>, min=<0 0>}
    }
    renderTransformPivot=<0.8 0.7>
   }
   slot {class=HorizontalBoxSlot, properties {horizontalAlignment=HAlign_Center, verticalAlignment=VAlign_Center}}
  }
  Overlay_Text_Test {
   widgetClassPath=&Overlay
   parent=Widget_ResourceCount
   widget=Overlay_Text_Test
   properties {renderTransform {angle=0, scale=<1 1>, shear=<-10 0>, translation=<0 0>}, renderTransformPivot=<0.2 0.7>}
   slot {class=HorizontalBoxSlot, properties {verticalAlignment=VAlign_Center}}
  }
  SizeBox_71 {
   widgetClassPath=&SizeBox
   parent=Overlay_Text_Test
   widget=SizeBox_71
   properties {bOverride_WidthOverride=true, widthOverride=100}
   slot {class=OverlaySlot, properties {horizontalAlignment=HAlign_Fill, verticalAlignment=VAlign_Fill}}
  }
  HorizontalBox_3 {
   widgetClassPath=&HorizontalBox
   parent=SizeBox_71
   widget=HorizontalBox_3
   slot {class=SizeBoxSlot, properties {horizontalAlignment=HAlign_Left, verticalAlignment=VAlign_Top}}
  }
  Text_CountSign {
   widgetClassPath=&CommonTextBlock
   parent=HorizontalBox_3
   widget=Text_CountSign
   properties {
    bSimpleTextMode=true
    text=+
    font {
     bForceMonospaced=false
     bMaterialIsStencil=false
     fontMaterial=null
     fontObject=&Burbank
     letterSpacing=0
     monospacedWidth=1
     size=32
     skewAmount=0
     typefaceFontName=Burbank-PlayerPower
     outlineSettings {
      bApplyOutlineToDropShadows=false
      bMiteredCorners=false
      bSeparateFillAlpha=false
      outlineMaterial=null
      outlineSize=0
      outlineColor=<rgba 0 0 0 1>
     }
    }
    shadowColorAndOpacity=<rgba 0 0 0 1>
    shadowOffset=<3 5>
   }
   slot {class=HorizontalBoxSlot, properties {horizontalAlignment=HAlign_Left, verticalAlignment=VAlign_Center}}
  }
  Text_ResourceCount {
   widgetClassPath=&CommonTextBlock
   bInherited=true
   parent=HorizontalBox_3
   widget=Text_ResourceCount
   properties {
    bSimpleTextMode=true
    text=100
    font {
     bForceMonospaced=false
     bMaterialIsStencil=false
     fontMaterial=&M_ResourceText
     fontObject=&Burbank
     letterSpacing=0
     monospacedWidth=1
     size=32
     skewAmount=0
     typefaceFontName=Burbank-Black
     outlineSettings {
      bApplyOutlineToDropShadows=false
      bMiteredCorners=false
      bSeparateFillAlpha=false
      outlineMaterial=null
      outlineSize=0
      outlineColor=<rgba 0 0 0 1>
     }
    }
    shadowColorAndOpacity=<rgba 0 0 0 1>
    shadowOffset=<3 5>
   }
   slot {class=HorizontalBoxSlot, properties {horizontalAlignment=HAlign_Left, verticalAlignment=VAlign_Center}}
  }
  Widget_Glow {
   widgetClassPath=&Image
   bInherited=true
   bIsVariable=true
   parent=Overlay_Text_Test
   widget=Widget_Glow
   properties {
    renderOpacity=0
    visibility=SelfHitTestInvisible
    brush {
     bIsDynamicallyLoaded=false
     drawAs=Image
     imageType=NoImage
     mirroring=NoMirror
     resourceName=None
     resourceObject=/Game/Athena/HUD/Materials/M_CountdownBG_SolidGlow
     tiling=NoTile
     imageSize=<170.1995 93.5933>
     margin=<ltrb 0 0 0 0>
     outlineSettings {
      bUseBrushTransparency=false
      roundingType=HalfHeightRadius
      width=0
      color {colorUseRule=UseColor_Specified, specifiedColor=<rgba 0 0 0 0>}
      cornerRadii=<0 0 0 0>
     }
     tintColor {colorUseRule=UseColor_Specified, specifiedColor=<rgba 0 0.2692 1 1>}
     uVRegion {bIsValid=false, max=<0 0>, min=<0 0>}
    }
    colorAndOpacity=<rgba 1 1 1 0.1429>
    renderTransform {angle=0, scale=<1 1>, shear=<-10 0>, translation=<2 0>}
   }
   slot {class=OverlaySlot, properties {horizontalAlignment=HAlign_Center, verticalAlignment=VAlign_Center, padding=<ltrb -250 -250 -250 -250>}}
  }
  TutorialHighlight {widgetClassPath=&TutorialHighlight}
 ]
}
